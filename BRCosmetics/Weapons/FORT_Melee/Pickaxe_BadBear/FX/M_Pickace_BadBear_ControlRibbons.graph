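# An unlit translucent material for Niagara ribbon effects on the BadBear pickaxe.
# Particle color RGB runs through the shared MF_FX_EmissiveAdjustments function to form the glow color.
# Particle alpha is multiplied by a period-2 sine of the UV's V coordinate, a texture-free soft mask that peaks mid-ribbon and fades to zero at both edges.
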Material /BRCosmetics/Weapons/FORT_Melee/Pickaxe_BadBear/FX/M_Pickace_BadBear_ControlRibbons
graph {
 properties {bUsedWithNiagaraRibbons=true, blendMode=BLEND_Translucent, pixelDepthOffsetMode=PDOM_Legacy, shadingModel=MSM_Unlit, usedShadingModels=Unlit}
 nodes [
  n1 = ParticleColor()
  MaterialFunctionCall(Emissive=@n1.RGB) {materialFunction=&MF_FX_EmissiveAdjustments}
  n2 = TextureCoordinate()
  Multiply(A=@n1.A, B=Sine(None=ComponentMask(None=@n2) {g=true}) {period=2})
 ]
}
