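# Physics asset for the BadBear pickaxe mesh, giving physics bodies to a three-bone chain named dyn_handle_01 to 03.
# Each bone carries one capsule, with a much larger, offset one on dyn_handle_03.
# Two constraints join neighboring bones in sequence.
# Both set swing and twist motion to Free, so their stored angle limits, wider on the second joint, do not restrict rotation.
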PhysicsAsset /BRCosmetics/Weapons/FORT_Melee/Pickaxe_BadBear/Meshes/BadBear_Axe_Physics
bodies {
 dyn_handle_01 [{shapeType=Capsule, length=10, radius=5, shapeName=dyn_handle_01_capsule, rotation=<rot 90 0 0>}]
 dyn_handle_02 [{shapeType=Capsule, length=10, radius=5, shapeName=dyn_handle_02_capsule, rotation=<rot 90 0 0>}]
 dyn_handle_03 [{
   shapeType=Capsule
   length=16.8465
   radius=19.1733
   shapeName=dyn_handle_03_capsule
   center=<0.1433 -2.0574 17.4731>
   rotation=<rot 90 -90.0002 -90.0219>
  }]
}
constraints [
 {
  bone1Name=dyn_handle_01
  bone2Name=dyn_handle_02
  swing1LimitDegrees=10
  swing1Motion=Free
  swing2LimitDegrees=10
  swing2Motion=Free
  twistLimitDegrees=5
  twistMotion=Free
 }
 {
  bone1Name=dyn_handle_02
  bone2Name=dyn_handle_03
  swing1LimitDegrees=15
  swing1Motion=Free
  swing2LimitDegrees=15
  swing2Motion=Free
  twistLimitDegrees=10
  twistMotion=Free
 }
]
