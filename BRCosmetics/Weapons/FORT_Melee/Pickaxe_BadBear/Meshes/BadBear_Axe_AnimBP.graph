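# Animation Blueprint for the BadBear pickaxe mesh that adds secondary physics motion to an incoming pose.
# A RigidBody node simulates the axe's physics asset in base-bone space with scaled, clamped component motion.
# Its output feeds displacement-limited, rotating spring controllers on dyn_bubble_2, then dyn_bubble_1, before the final pose.
# A second three-bone spring chain is left unconnected, and the event graph holds no logic.
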AnimBlueprint /BRCosmetics/Weapons/FORT_Melee/Pickaxe_BadBear/Meshes/BadBear_Axe_AnimBP
target_skeleton=&/BadBear_Axe_Skeleton
graphs [
 AnimGraph {nodes [
   AnimGraphNode_SubInput_0 = Input Pose()
   n1 = RigidBody(ComponentPose=Local To Component(LocalPose=@AnimGraphNode_SubInput_0)) {
    bClampLinearTranslationLimitToRefPose=true
    bOverrideWorldGravity=true
    overridePhysicsAsset=&/BadBear_Axe_Physics
    simulationSpace=BaseBoneSpace
    alphaBoolBlend {blendOption=Linear}
    componentAppliedLinearAccClamp=<4000 4000 4000>
    componentLinearAccScale=<0.2 0.2 0.2>
    componentLinearVelScale=<0.2 0.2 0.2>
   }
   n2 = Spring controller - Bone: dyn_bubble_2!(ComponentPose:Component Space Pose Link Structure=@n1, Alpha:float=1, SpringStiffness:double=700, SpringDamping:double=700) -> (Pose:Component Space Pose Link Structure) {
    is=title AnimGraphNode_SpringBone
    bLimitDisplacement=true
    bRotateX=true
    bRotateY=true
    bRotateZ=true
    maxDisplacement=1.5
    springDamping=10
    springStiffness=10
    alphaBoolBlend {blendOption=Linear}
    springBone {boneName=dyn_bubble_2}
   }
   n3 = Spring controller - Bone: dyn_bubble_1!(ComponentPose:Component Space Pose Link Structure=@n2, Alpha:float=1, SpringStiffness:double=700, SpringDamping:double=700) -> (Pose:Component Space Pose Link Structure) {
    is=title AnimGraphNode_SpringBone
    bLimitDisplacement=true
    bRotateX=true
    bRotateY=true
    bRotateZ=true
    maxDisplacement=2
    springDamping=10
    springStiffness=2.5
    alphaBoolBlend {blendOption=Linear}
    springBone {boneName=dyn_bubble_1}
   }
   Output Pose(Result=Component To Local(ComponentPose=@n3)) {layerGroup=None}
   n4 = Spring controller - Bone: dyn_bubble_2!(ComponentPose, Alpha=1) -> (Pose) {
    bLimitDisplacement=true
    bRotateX=true
    bRotateY=true
    bRotateZ=true
    maxDisplacement=1
    springDamping=10
    springStiffness=1
    alphaBoolBlend {blendOption=Linear}
    springBone {boneName=dyn_bubble_2}
   }
   n5 = Spring controller - Bone: dyn_bubble_3!(ComponentPose:Component Space Pose Link Structure=@n4, Alpha:float=1) -> (Pose:Component Space Pose Link Structure) {
    is=title AnimGraphNode_SpringBone
    bLimitDisplacement=true
    bRotateX=true
    bRotateY=true
    bRotateZ=true
    maxDisplacement=1
    springDamping=10
    springStiffness=1
    alphaBoolBlend {blendOption=Linear}
    springBone {boneName=dyn_bubble_3}
   }
   Spring controller - Bone: dyn_bubble_1!(ComponentPose=@n5, Alpha=1) -> (Pose) {
    bLimitDisplacement=true
    bRotateX=true
    bRotateY=true
    bRotateZ=true
    maxDisplacement=1
    springDamping=10
    springStiffness=1
    alphaBoolBlend {blendOption=Linear}
    springBone {boneName=dyn_bubble_1}
   }
  ]}
 EventGraph {nodes [
   TryGetPawnOwner()
   Event BlueprintUpdateAnimation()
  ]}
]
variables []
